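# Niagara system for the CrumbViolin back bling, adding a persistent green glow to the backpack cosmetic.
# LightAndGlow bursts a particle driving an offset light and camera-offset sprite.
# MeshGlows bursts two particles, using spawn groups and visibility tags to vary color and offset across two mesh renderers.
# System-level ReactivityFlicker and ToD values multiply into both emitters' brightness or alpha, with curve, sine, time-of-day and front-end inputs among the dependencies.
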NiagaraSystem /BRCosmetics/Accessories/FORT_Backpacks/Backpack_M_MED_CrumbViolin/FX/NS_Backpack_CrumbViolin
dependencies {
 usedDataInterfaces [&NiagaraDataInterfaceCurve]
 usedDynamicInputs [
  &FloatFromCurve
  &Lerp_Float
  &Frac_Float
  &Sine
  &FNTimeOfDayScalar
  &FN_FrontEndFloatModifier
 ]
 usedModules [
  &SystemState
  &self:SystemSpawnScript.SystemScriptSource.SystemScriptGraph.NiagaraNodeAssignment_1.SetVariables_#1
  &self:SystemSpawnScript.SystemScriptSource.SystemScriptGraph.NiagaraNodeAssignment_0.SetVariables_#2
 ]
}
emitters {
 LightAndGlow {
  emitterUpdateScript [
   &EmitterState(
    Inactive Response=::NewEnumerator1
    Life Cycle Mode=::NewEnumerator1
    Loop Behavior=::NewEnumerator1
    Loop Duration Mode=::NewEnumerator1
    UseLoopDelay=false
   )
   &SpawnBurst_Instantaneous(Use Spawn Probability=false, UseLoopCountLimit=true)
  ]
  particleSpawnScript [
   &InitializeParticle(
    Color=<rgba 0.0833 1 0 0.6>
    Mesh Scale Mode=::NewEnumerator3
    Mesh Uniform Scale=3
    Sprite Size Mode=::NewEnumerator3
    Uniform Sprite Size=28
    UsePositionOffset=true
    Write Material Random=true
   )
   &DynamicMaterialParameters(
    Index 0 Param 1=1.6
    Index 0 Param 2=0.2
    Module.Index0Float0Write=true
    Module.Index0Float1Write=true
    Module.Index0Float2Write=true
    Module.Index0Float3Write=true
   )
  ]
  particleUpdateScript [
   &ParticleState(Kill Particles When Lifetime Has Elapsed=false)
   &ScaleColor(ScaleA=true, ScaleRGB=true)
   &FN_Light_Attributes(
    Color=&Multiply_LinearColorByFloat(Linear Color=<rgba 0.0833 1 0 0.6>, Float=35)
    Create Color Attribute=true
    Create Exponent Attribute=true
    Create Position Attribute=true
    Create Radius Attribute=true
    Exponent=2
    Position=&AddVectorToPosition(Position=$Particles.Position, Vector=<0 -25 -5>)
    Radius=19
    Scale Color=&Multiply_Float(A=$System.ReactivityFlicker, B=$System.ToD)
   )
   &FN_Sprite_Attributes(
    Camera Offset=5
    Color=&MultiplyLinearColor(
     Linear Color=<rgba 4 20 0 1>
     Scale Factor=&MakeVector4(X=1, Y=1, Z=1, W=&Multiply_Float(A=$System.ReactivityFlicker, B=$System.ToD))
    )
    Create Camera Offset Attribute=true
    Create Color Attribute=true
    Create Position Attribute=true
    Create SpriteSize Attribute=true
    Position:NiagaraPosition=&AddVectorToPosition(Position=$Particles.Position, Vector=<0 -15 -5>)
    SpriteSize=<22 22>
    Write to Module Namespace=true
   )
  ]
  rendererClasses [&NiagaraLightRendererProperties, &NiagaraSpriteRendererProperties]
  renderers [
   {rendererClass=&NiagaraLightRendererProperties, rendererIndex=0}
   {rendererClass=&NiagaraSpriteRendererProperties, rendererIndex=1}
  ]
 }
 MeshGlows {
  emitterUpdateScript [
   &EmitterState(
    Inactive Response=::NewEnumerator1
    Life Cycle Mode=::NewEnumerator1
    Loop Behavior=::NewEnumerator1
    Loop Duration Mode=::NewEnumerator1
    UseLoopDelay=false
   )
   &SpawnBurst_Instantaneous(Spawn Count=2, Use Spawn Probability=false, UseLoopCountLimit=true)
  ]
  particleSpawnScript [
   &FN_SetSpawnGroup(Number of Spawn Groups=2)
   &InitializeParticle(
    Color=&MergeHelper_ColorSwitch(0=<rgba 0.0833 1 0 0.333>, 1=<rgba 0.0833 1 0 1.6>, 2=?, 3=?, Index=$Particles.SpawnGroup, Num=::NewEnumerator0)
    Lifetime=5
    Mesh Scale Mode=::NewEnumerator3
    Position Offset=&MergeHelper_VectorSwitch(0=<0 0 3>, 1=<0 0.5 0>, 2=?, 3=?, Index=$Particles.SpawnGroup, Num=::NewEnumerator0)
    UsePositionOffset=true
    Write Material Random=true
   )
   /SetVariables(Particles.VisibilityTag=&MergeHelper_IntSwitch(0=0, 1=1, 2=?, 3=?, 4=?, 5=?, Index=$Particles.SpawnGroup, Num=::NewEnumerator0)) {isSetParameters=true}
   &DynamicMaterialParameters(
    Index 0 Param 1=1.5
    Index 0 Param 2=0.01
    Index 0 Param 3=&MergeHelper_FloatSwitch(0=1, 1=1, 2=?, 3=?, Index=$Particles.SpawnGroup, Num=::NewEnumerator0)
    Index 0 Param 4=1
    Module.Index0Float0Write=true
    Module.Index0Float1Write=true
    Module.Index0Float2Write=true
    Module.Index0Float3Write=true
   )
  ]
  particleUpdateScript [
   &ParticleState(Kill Particles When Lifetime Has Elapsed=false)
   &CameraOffset()
   &ScaleColor(Scale Alpha=&Multiply_Float(A=$System.ReactivityFlicker, B=$System.ToD), ScaleA=true, ScaleRGB=true)
  ]
  rendererClasses [&NiagaraMeshRendererProperties]
  renderers [
   {rendererClass=&NiagaraMeshRendererProperties, rendererIndex=0}
   {rendererClass=&NiagaraMeshRendererProperties, rendererIndex=1}
  ]
 }
}
summary {
 systemName=NS_Backpack_CrumbViolin
 emitters [
  {emitterName=LightAndGlow, rendererClasses [&NiagaraLightRendererProperties, &NiagaraSpriteRendererProperties]}
  {emitterName=MeshGlows, rendererClasses [&NiagaraMeshRendererProperties]}
 ]
 userVariables [
  User.bIsFrontEnd {type=NiagaraBool}
  User.bIsFrontEndPreview {type=NiagaraBool}
  User.IsFlicker? {type=NiagaraFloat, defaultValue=0, description=""}
 ]
}
