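# Material function for the Crumb Violin backpack cosmetic that overrides emissive, opacity mask and specular on incoming material attributes to draw a glowing portal and arm.
# Distorted, scrolling caustic height samples, one bump-offset, are overlay-blended and power-shaped into fake raymarch noise.
# Packed masks supply glow, distance-field outline and hand/nail regions, with emissive colored through a gradient lookup.
# Local-position plane masks drive portal masking and arm glow fade, and opacity uses temporal dithering.
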MaterialFunction /BRCosmetics/Accessories/FORT_Backpacks/Backpack_M_MED_CrumbViolin/Materials/MF_CrumbViolin_Backpack
graph {nodes [
  n1 = GetMaterialAttributes(None=FunctionInput() {inputName=Result, inputType=FunctionInput_MaterialAttributes}) {attributeGetTypes [#1]}
  n2 = NamedRerouteUsage()
  FunctionOutput(None=SetMaterialAttributes!(
    MaterialAttributes=@n1.MaterialAttributes
    EmissiveColor=NamedRerouteUsage()
    OpacityMask=NamedRerouteUsage()
    Specular=LinearInterpolate(A=@n1.Specular, Alpha=@n2) {constB=0}
   ) {attributeSetTypes [#2, #3, #1]}) {outputName=SetMaterialAttributes}
  n3 = ScalarParameter() {defaultValue=0.03, group=Backpack - Crumb Violin, parameterName=Outline - Size}
  n4 = SmoothStep(Min=@n3, Max=Add(A=@n3, B=ScalarParameter() {group=Backpack - Crumb Violin, parameterName=Outline - Fade Amount}), Value=NamedRerouteUsage()) {constMax=0.15}
  NamedRerouteDeclaration(None=Multiply(A=OneMinus(None=@n4), B=NamedRerouteUsage())) {name=Glow Outline, nodeColor=<rgba 1 0.5625 0 1>}
  n5 = Time()
  n6 = Frac(None=Sine(None=Multiply(A=@n5)))
  n7 = ScalarParameter() {group=Backpack - Crumb Violin, parameterName=IsFlicker?}
  NamedRerouteDeclaration(None=Multiply(A=Multiply(A=@n6, B=@n7), B=ScalarParameter() {defaultValue=1, group=Backpack - Crumb Violin, parameterName=Reactivity on?})) {name=Flicker, nodeColor=<rgba 0 0.8125 1 1>}
  n8 = TextureSampleParameter2D() {
   group=Backpack - Crumb Violin
   parameterName=Crumb Violin - Mask
   samplerType=SAMPLERTYPE_Masks
   texture=/BRCosmetics/Accessories/FORT_Backpacks/Backpack_M_MED_CrumbViolin/Textures/T_CrumbViolin_Backpack_Mask
  }
  NamedRerouteDeclaration(None=@n8.R) {name=Hand Mask, nodeColor=<rgba 0.1016 0 1 1>}
  n9 = Min(A=@n8.R, B=@n8.R)
  NamedRerouteDeclaration(None=Subtract(A=@n8.B, B=@n9)) {name=Glow Mask, nodeColor=<rgba 1 0.2578 0 1>}
  NamedRerouteDeclaration(None=@n8.G) {name=DF Mask, nodeColor=<rgba 1 0.3281 0 1>}
  NamedRerouteDeclaration(None=ComponentMask(None=FunctionInput() {inputName=Tech Art Mask, sortPriority=10}) {b=true}) {name=Portal, nodeColor=<rgba 0.125 1 0 1>}
  NamedRerouteDeclaration(None=@n9) {name=Nail Mask, nodeColor=<rgba 1 0 0.9141 1>}
  n10 = NamedRerouteUsage()
  n11 = Max(A=OneMinus(None=NamedRerouteUsage()), B=OneMinus(None=Step(X=@n10) {constX=0.01, constY=0.1}))
  NamedRerouteDeclaration(None=Saturate(None=MaterialFunctionCall(Alpha Threshold=@n11) {materialFunction=&DitherTemporalAA}.Result)) {name=Opacity Out, nodeColor=<rgba 1 0 0.2109 1>}
  n12 = Multiply(
   A=Power(Base=NamedRerouteUsage(), Exp=ScalarParameter() {defaultValue=4, group=Backpack - Crumb Violin, parameterName=Glow - Fade})
   B=ScalarParameter() {defaultValue=0.5, group=Backpack - Crumb Violin, parameterName=Glow - Intensity}
  )
  n13 = NamedRerouteUsage()
  n14 = NamedRerouteUsage()
  n15 = NamedRerouteUsage()
  n16 = LinearInterpolate(
   A=Add(A=NamedRerouteDeclaration(None=@n12) {name=Outer Glow, nodeColor=<rgba 0 0.1328 1 1>}, B=Multiply(A=@n13, B=@n14))
   B=Multiply(A=@n15, B=VectorParameter() {group=Backpack - Crumb Violin, parameterName=Outline Color, defaultValue=<rgba 0.3269 1.2 0.2062 1>}.RGB)
   Alpha=Multiply(A=@n15, B=ScalarParameter() {defaultValue=0.5, group=Backpack - Crumb Violin, parameterName=Glow Outline - Strength}) {constB=0.15}
  )
  n17 = TextureSampleParameter2D(UVs=Saturate(None=ComponentMask(None=@n16) {g=true, r=true})) {
   group=Backpack - Crumb Violin
   parameterName=Color Gradient
   samplerType=SAMPLERTYPE_LinearColor
   texture=/BRCosmetics/Accessories/FORT_Backpacks/Backpack_M_MED_CrumbViolin/Textures/CA_CrumbViolin_Gradients
  }
  n18 = Multiply(A=@n16, B=@n17.RGB)
  n19 = ScalarParameter() {defaultValue=0.95, group=Backpack - Crumb Violin, parameterName=Emissive - Strength}
  NamedRerouteDeclaration(None=LinearInterpolate(
    A=Multiply(A=@n18, B=@n19)
    B=VectorParameter() {group=Backpack - Crumb Violin, parameterName=Arm Glow Color, defaultValue=<rgba 0.0483 0.5625 0 1>}.RGB
    Alpha=NamedRerouteUsage()
   )) {name=Emissive Out, nodeColor=<rgba 0.8516 1 0 1>}
  n20 = Multiply(
   A=TextureCoordinate()
   B=MaterialFunctionCall(
    X=ScalarParameter() {defaultValue=2.5, group=Backpack - Crumb Violin, parameterName=Noise - Tiling - X}
    Y=ScalarParameter() {defaultValue=2, group=Backpack - Crumb Violin, parameterName=Noise - Tiling - Y}
   ) {materialFunction=&MakeFloat2}.Result
  )
  n21 = Time()
  n22 = ScalarParameter() {defaultValue=1.5, group=Backpack - Crumb Violin, parameterName=Noise - Height - Speed Multipler}
  n23 = VectorParameter() {group=Backpack - Crumb Violin, parameterName=Noise - Distortion - Speed, defaultValue=<rgba 0 0.04 0 1>}
  n24 = Multiply(A=@n21, B=ComponentMask(None=Multiply(A=@n22, B=@n23.RGB)) {g=true, r=true})
  n25 = Add(A=@n20, B=@n24)
  n26 = ComponentMask(None=Multiply(A=VectorParameter() {group=Backpack - Crumb Violin, parameterName=Noise - Height - Speed, defaultValue=<rgba -0.02 0 0 1>}.RGB, B=@n22)) {g=true, r=true}
  n27 = TextureSample(UVs=Add(A=@n20, B=Multiply(A=@n26, B=@n21))) {
   samplerType=SAMPLERTYPE_LinearGrayscale
   texture=/BRCosmetics/Accessories/FORT_Backpacks/Backpack_M_MED_CrumbViolin/Textures/ChromeTech_CausticsSourceHeight
  }
  n28 = Multiply(
   A=ComponentMask(None=@n27.RGB) {g=true, r=true}
   B=ScalarParameter() {defaultValue=0.1, group=Backpack - Crumb Violin, parameterName=Noise - Distortion - Strength}
  )
  n29 = TextureSample(UVs=BumpOffset(
    Coordinate=Add(A=@n25, B=@n28)
    Height=ScalarParameter() {defaultValue=-5, group=Backpack - Crumb Violin, parameterName=Noise - Bump Offset - Depth}
   )) {
   samplerType=SAMPLERTYPE_LinearGrayscale
   texture=&ChromeTech_CausticsSourceHeight
  }
  n30 = MaterialFunctionCall(Float2=Add(A=@n20, B=OneMinus(None=@n24))) {materialFunction=&BreakOutFloat2Components}
  n31 = Add(A=@n30.R) {constB=0.3}
  n32 = TextureSample(UVs=Add(A=MaterialFunctionCall(X=@n31, Y=@n30.G) {materialFunction=&MakeFloat2}.Result, B=@n28)) {
   samplerType=SAMPLERTYPE_LinearGrayscale
   texture=&ChromeTech_CausticsSourceHeight
  }
  NamedRerouteDeclaration(None=MaterialFunctionCall(Base=@n29.RGB, Blend=@n32.RGB) {materialFunction=&Blend_Overlay}.Result) {name=Raymarch Tex, nodeColor=<rgba 0.8281 0 1 1>}
  n33 = Add(
   A=LocalPosition().XYZ
   B=VectorParameter() {group=Backpack - Crumb Violin, parameterName=Portal Center Pos, defaultValue=<rgba 0 10 0 0>}.RGB
  )
  n34 = Constant3Vector() {constant=<rgba 0 1 0 0>}
  NamedRerouteDeclaration(None=SmoothStep(
    Max=ScalarParameter() {defaultValue=1, group=Backpack - Crumb Violin, parameterName=Arm - Masking - Smoothness}
    Value=DotProduct(A=@n33, B=@n34)
   ) {constMax=2}) {name=PortalMask, nodeColor=<rgba 0 0.9062 1 1>}
  n35 = Add(
   A=LocalPosition().XYZ
   B=VectorParameter() {group=Backpack - Crumb Violin, parameterName=Portal Center Pos 2, defaultValue=<rgba 0 14 0 0>}.RGB
  )
  n36 = Constant3Vector() {constant=<rgba 0 1 0 0>}
  NamedRerouteDeclaration(None=SmoothStep(
    Max=ScalarParameter() {defaultValue=30, group=Backpack - Crumb Violin, parameterName=Arm - Glow - Fade Distance}
    Value=DotProduct(A=@n35, B=@n36)
   ) {constMax=30}) {name=Portal Glow Mask, nodeColor=<rgba 0 0.9531 1 1>}
  n37 = SmoothStep(
   Max=ScalarParameter() {defaultValue=1.2, group=Backpack - Crumb Violin, parameterName=Noise - Smoothness}
   Value=OneMinus(None=NamedRerouteUsage())
  )
  n38 = Power(Base=@n37, Exp=ScalarParameter() {defaultValue=6.0538, group=Backpack - Crumb Violin, parameterName=Noise - Mid - Power})
  n39 = ScalarParameter() {defaultValue=5, group=Backpack - Crumb Violin, parameterName=Noise - Mid - Intensity}
  n40 = Power(
   Base=Max(A=Power(Base=@n37) {constExponent=1}, B=Multiply(A=@n38, B=@n39))
   Exp=ScalarParameter() {defaultValue=0.5, group=Backpack - Crumb Violin, parameterName=Noise - Low - Power}
  ) {constExponent=0.6}
  n41 = ScalarParameter() {defaultValue=0.4, group=Backpack - Crumb Violin, parameterName=Noise - Low - Intensity}
  n42 = Power(Base=@n37, Exp=ScalarParameter() {defaultValue=15, group=Backpack - Crumb Violin, parameterName=Noise - Highlights - Power})
  n43 = ScalarParameter() {defaultValue=18, group=Backpack - Crumb Violin, parameterName=Noise - Highlights - Max Intensity}
  NamedRerouteDeclaration(None=Max(A=Max(A=@n40, B=@n41), B=Multiply(A=@n42, B=@n43))) {name=Raymarch Noise, nodeColor=<rgba 0.4531 1 0 1>}
  n44 = Multiply(A=@n10, B=NamedRerouteUsage())
  n45 = ScalarParameter() {defaultValue=1, group=Backpack - Crumb Violin, parameterName=Arm Glow Strength}
  NamedRerouteDeclaration(None=Subtract(A=Multiply(A=@n44, B=@n45), B=NamedRerouteUsage())) {name=Arm Glow, nodeColor=<rgba 0.0781 1 0 1>}
  NamedRerouteUsage()
  OneMinus(None=@n26)
 ]}
