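# Anim Blueprint for the CrumbViolin cosmetic backpack, playing an idle with occasional random idle breaks and spinning the dyn_antenna_01 bone.
# A state machine feeds additive bone transforms that offset the pack and apply a continuously incremented roll to the antenna.
# The owner's spine_05-to-spine_01 yaw delta weights an extra pack offset and tilt.
# Random delays, shorter in frontend, trigger one of two breaks; post-evaluate clears the trigger.
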AnimBlueprint /BRCosmetics/Accessories/FORT_Backpacks/Backpack_M_MED_CrumbViolin/Meshes/M_MED_CrumbViolin_Pack_AnimBP
target_skeleton=&/M_MED_CrumbViolin_Pack_Skeleton
graphs [
 AnimGraph {nodes [
   Apply Additive(Base=Sequence Player 'CrumbViolin_Idle_BASE_PACK'() {sequence=/BRCosmetics/Accessories/FORT_Backpacks/Backpack_M_MED_CrumbViolin/Animations/CrumbViolin_Idle_BASE_PACK}) {alphaBoolBlend {blendOption=Linear}}
   Input Pose()
   Random Sequence Player!() -> (Pose:Pose Link Structure) {
    is=title AnimGraphNode_RandomPlayer
    entries [
     {
      chanceToPlay=0.65
      maxLoopCount=0
      maxPlayRate=1
      minLoopCount=0
      minPlayRate=1
      sequence=/BRCosmetics/Accessories/FORT_Backpacks/Backpack_M_MED_CrumbViolin/Animations/CrumbViolin_Idle_PACK
      blendIn {blendOption=Linear, blendTime=0, customCurve=null}
     }
     {
      chanceToPlay=0.15
      maxLoopCount=0
      maxPlayRate=1
      minLoopCount=0
      minPlayRate=1
      sequence=/BRCosmetics/Accessories/FORT_Backpacks/Backpack_M_MED_CrumbViolin/Animations/CrumbViolin_Idle_01_PACK
      blendIn {blendOption=Linear, blendTime=0, customCurve=null}
     }
     {
      chanceToPlay=0.2
      maxLoopCount=0
      maxPlayRate=1
      minLoopCount=0
      minPlayRate=1
      sequence=/BRCosmetics/Accessories/FORT_Backpacks/Backpack_M_MED_CrumbViolin/Animations/CrumbViolin_Idle_02_PACK
      blendIn {blendOption=Linear, blendTime=0, customCurve=null}
     }
    ]
   }
   AnimGraphNode_StateMachine_0 = CrumbViolin state Machine() {is=title AnimGraphNode_StateMachine}
   n1 = "Transform (Modify) Bone - Bone: pack"(ComponentPose=Local To Component(LocalPose=@AnimGraphNode_StateMachine_0), Translation="0,-2,0", Rotation="0,0,0") {
    rotationMode=BMM_Additive
    rotationSpace=BCS_BoneSpace
    translationMode=BMM_Additive
    alphaBoolBlend {blendOption=Linear}
    boneToModify {boneName=pack}
    translation {y=-2}
   }
   n2 = "Transform (Modify) Bone - Bone: dyn_antenna_01"(ComponentPose=@n1, Alpha=1, Translation="0,0,0", Rotation=Get ball_rotator() -> (ball_rotator:Rotator), Scale="1,1,1") {rotationMode=BMM_Additive, rotationSpace=BCS_BoneSpace, alphaBoolBlend {blendOption=Linear}, boneToModify {boneName=dyn_antenna_01}}
   n3 = "Transform (Modify) Bone - Bone: pack"(ComponentPose=@n2, Alpha=Get BackpackRotate() -> (BackpackRotate:double), Translation="4,-6,8", Rotation="0,0,30") {
    rotationMode=BMM_Additive
    rotationSpace=BCS_BoneSpace
    scaleMode=BMM_Additive
    scaleSpace=BCS_BoneSpace
    translationMode=BMM_Additive
    translationSpace=BCS_BoneSpace
    alphaBoolBlend {blendOption=Linear}
    boneToModify {boneName=pack}
    rotation {roll=30}
    translation=<4 -6 8>
   }
   Output Pose(Result=Component To Local(ComponentPose=@n3)) {layerGroup=None}
  ]}
 CrumbViolin state Machine {
  entry=Idle
  path=AnimGraph.AnimGraphNode_StateMachine_0.CrumbViolin state Machine
  ids=id
  nodes [
   AnimStateNode_0 = Idle()
   AnimStateNode_1 = IdleBreak_01() {is=title AnimStateNode}
   AnimStateNode_2 = IdleBreak_02() {is=title AnimStateNode}
   AnimStateTransitionNode_0 = AnimStateTransitionNode()
   AnimStateTransitionNode_1 = AnimStateTransitionNode()
   AnimStateTransitionNode_2 = AnimStateTransitionNode() {bAutomaticRuleBasedOnSequencePlayerInState=true}
   AnimStateTransitionNode_3 = AnimStateTransitionNode() {bAutomaticRuleBasedOnSequencePlayerInState=true}
  ]
  links [
   {
    dst=IdleBreak_01
    rule=AnimGraph.AnimGraphNode_StateMachine_0.CrumbViolin state Machine.AnimStateTransitionNode_0.Transition
    src=Idle
    via=AnimStateTransitionNode_0
   }
   {
    dst=IdleBreak_02
    rule=AnimGraph.AnimGraphNode_StateMachine_0.CrumbViolin state Machine.AnimStateTransitionNode_1.Transition
    src=Idle
    via=AnimStateTransitionNode_1
   }
   {
    dst=Idle
    rule=AnimGraph.AnimGraphNode_StateMachine_0.CrumbViolin state Machine.AnimStateTransitionNode_2.Transition
    src=IdleBreak_01
    via=AnimStateTransitionNode_2
   }
   {
    dst=Idle
    rule=AnimGraph.AnimGraphNode_StateMachine_0.CrumbViolin state Machine.AnimStateTransitionNode_3.Transition
    src=IdleBreak_02
    via=AnimStateTransitionNode_3
   }
  ]
 }
 Idle {
  path=AnimGraph.AnimGraphNode_StateMachine_0.CrumbViolin state Machine.AnimStateNode_0.Idle
  nodes [Output Animation Pose(Result=Sequence Player 'CrumbViolin_Idle_PACK'() {bLoopAnimation=false, sequence=&CrumbViolin_Idle_PACK})]
 }
 IdleBreak_01 {
  path=AnimGraph.AnimGraphNode_StateMachine_0.CrumbViolin state Machine.AnimStateNode_1.IdleBreak_01
  nodes [
   n1 = Sequence Player 'CrumbViolin_Idle_01_PACK'() {bLoopAnimation=false, sequence=&CrumbViolin_Idle_01_PACK}
   Output Animation Pose(Result=@n1)
  ]
 }
 IdleBreak_02 {
  path=AnimGraph.AnimGraphNode_StateMachine_0.CrumbViolin state Machine.AnimStateNode_2.IdleBreak_02
  nodes [
   n1 = Sequence Player 'CrumbViolin_Idle_02_PACK'() {bLoopAnimation=false, sequence=&CrumbViolin_Idle_02_PACK}
   Output Animation Pose(Result=@n1)
  ]
 }
 Transition {
  path=AnimGraph.AnimGraphNode_StateMachine_0.CrumbViolin state Machine.AnimStateTransitionNode_0.Transition
  nodes [Result(bCanEnterTransition=AND Boolean(A=Get IdleBreak_01() -> (IdleBreak_01:bool), B=Get IdleAnim() -> (IdleAnim:bool)))]
 }
 Transition {
  path=AnimGraph.AnimGraphNode_StateMachine_0.CrumbViolin state Machine.AnimStateTransitionNode_1.Transition
  nodes [
   n1 = Get IdleAnim()
   Result(bCanEnterTransition=AND Boolean(A=Get IdleBreak_01(), B=NOT Boolean(A=@n1)))
  ]
 }
 Transition {path=AnimGraph.AnimGraphNode_StateMachine_0.CrumbViolin state Machine.AnimStateTransitionNode_2.Transition, nodes [Result()]}
 Transition {path=AnimGraph.AnimGraphNode_StateMachine_0.CrumbViolin state Machine.AnimStateTransitionNode_3.Transition, nodes [Result()]}
 EventGraph {nodes [
   Event BlueprintUpdateAnimation()
   > n1 = Sequence!(execute) -> (then_0, then_1, then_2:Exec)
   n2 = Get Spin_Int() -> (Spin_Int:double)
   > Set Spin_Int(Spin_Int:double=float + float(A=@n2, B=1.2) {is=title K2Node_CommutativeAssociativeBinaryOperator}) -> (Output_Get:double)
   > n3 = Branch(Condition="float >= float"(A=@n2, B=360) {is=title K2Node_CallFunction})
   > n4 = Set Spin_Int(Spin_Int=float - float(A=Get Spin_Int(), B=360) {is=title K2Node_CallFunction})
   n5 = Set Spin_Int(execute=@n3.else, Spin_Int=float + float(A=Get Spin_Int(), B=1.2))
   Set ball_rotator(execute=[@n4.then, @n5.then], ball_rotator:Rotator=MakeRotator(Roll=float * float(A=@n5, B=-1))) -> (Output_Get:Rotator)
   n6 = BreakTransform(InTransform=GetDeltaTransformFromRefPose(self=GetOwnerSkeletalMeshComponent(), BoneName=spine_05, BaseName=spine_01))
   Set BackpackRotate(execute=@n1.then_1, BackpackRotate:double=MapRangeClamped(Value=BreakRotator(InRot=@n6.Rotation).Yaw, InRangeA=-25, InRangeB=-65)) -> (Output_Get:double)
   n7 = Branch(execute=@n1.then_2, Condition=Get bIsInFrontEnd())
   > n8 = Delay(Duration=RandomFloatInRange(Min=5, Max=25))
   n9 = Set IdleBreak_01(execute=[@n11.then, @n8.then], IdleBreak_01:bool=true) -> (Output_Get:bool)
   > Branch(Condition=@n9)
   > n10 = Branch(Condition=RandomBool!() -> (ReturnValue:bool))
   > Set IdleAnim(IdleAnim:bool=false) -> (Output_Get:bool)
   Set IdleAnim(execute=@n10.else, IdleAnim=true)
   n11 = Delay(execute=@n7.else, Duration=RandomFloatInRange(Min=10, Max=40))
   Event BlueprintPostEvaluateAnimation() {is=title K2Node_Event}
   > Set IdleBreak_01(IdleBreak_01=false)
   TryGetPawnOwner()
   Reroute Node(InputPin:Rotator=@n6.Rotation) -> (OutputPin:Rotator)
   Get IdleAnim()
  ]}
]
variables [
 BackpackRotate {type=number}
 ball_rotator {type=Rotator}
 IdleAnim {type=boolean}
 IdleBreak_01 {type=boolean}
 Spin_Int {type=number}
]
